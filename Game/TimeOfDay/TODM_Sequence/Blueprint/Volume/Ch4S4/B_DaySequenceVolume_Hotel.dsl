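; B_DaySequenceVolume_Hotel is a day sequence volume Blueprint in the time-of-day system, filed under chapter 4 stage 4 content for the hotel.
; It overrides UpdateScalabilitySequence and the construction script, and both only call the parent implementation.
; The graphs add no logic of their own, so all behavior comes from the parent class.
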(fn UpdateScalabilitySequence
  (|CallParentFunction))

(fn ConstructionScript
  (|CallParentFunction))
