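; A day sequence volume that switches a heist base's lockdown time-of-day look on and off.
; When the playlist is ready, the authority listens for lockdown-initiated gameplay messages, sets EnableVolume from the heist base tag match and flushes dormancy.
; EnableVolume's OnRep drives a timeline fading the modifier's blend weight; fade-in requires the safe zone, and reverse completion disables the volume.
; Safe-zone changes also trigger fades, though the bind and unbind events show no binding logic.
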(event EventBeginPlay
  (|CallParentFunction)
  (Networking|SwitchHasAuthority
    (:Authority
      (Class|FortAsyncActionCurrentPlaylistReady|PlaylistDataReadyAsync
        (:OnReady
          (bind (_asynctaskproxy _payload) (Messaging|RegisterGameplayMessageReceiver (Variables|Lockdown|GetLockdownInitiatedChannelTag) "/Script/GameplayAbilities.GameplayEventData")
            (:OnMessageReceived
              (Variables|Default|SetEnableVolume (GameplayTags|HasTag (.TargetTags (Utilities|Struct|BreakGameplayEventData _payload)) (Variables|Lockdown|GetHeistBaseIdTagConst) true))
              (Networking|FlushNetDormancy))))))
    (:Remote
      ;[1 1 1] Force lockdown to be off on first begin play
      (Utilities|FlowControl|DoOnce false
        (:Completed
          (CallFunction|DisableDaysequenceVolume))))))

(event Custom|SetLockdownDaySeqVol
  (params (EnableVolume :type Boolean))
  (Variables|Default|SetEnableVolume EnableVolume))

(event Custom|FadeLockdownIn
  (if (Variables|Default|GetIsInSafeZone)
    (CallFunction|EnableDaysequenceVolume)
    (Variables|/Script/DaySequence.DaySequenceModifierComponent|SetUserBlendWeight 0.0 (Variables|Default|GetDaySequenceModifier))
    (bind (_direction _fade) (Timeline|FadeInTL PlayFromStart)
      (:Update
        (Variables|/Script/DaySequence.DaySequenceModifierComponent|SetUserBlendWeight _fade (Variables|Default|GetDaySequenceModifier)))
      (:Finished
        (switch Utilities|FlowControl|Switch|SwitchonETimelineDirection _direction
          (:Backward
            (CallFunction|DisableDaysequenceVolume)))))))

(event Custom|FadeLockdownOut
  (Timeline|FadeInTL ReverseFromEnd))

(event Custom|BindSafeZone
  (Variables|Default|GetCurrentViewTarget
    (:then)
    (:else)))

(event Custom|UnbindSafeZone
  (Variables|Default|GetCurrentViewTarget
    (:then)
    (:else)))

(event Custom|SafeZoneStateChanged
  (params (PlayerPawn :type Fort%20Player%20Pawn%20Object%20Reference) (bIsInSafeZone :type Boolean))
  ;[1 1 1] Make sure we turn this off if the player goes in the storm
  (Utilities|IsValid PlayerPawn
    (:Is%20Valid
      (bind _output_get (Variables|Default|SetIsInSafeZone bIsInSafeZone))
      (if _output_get
        (CallFunction|FadeLockdownIn)
        (else
          (CallFunction|FadeLockdownOut))))))

(fn OnRep_EnableVolume
  (Networking|CanExecuteCosmeticEvents
    (:True
      (if (Variables|Default|GetEnableVolume)
        (CallFunction|BindSafeZone)
        (CallFunction|FadeLockdownIn)
        (else
          (CallFunction|UnbindSafeZone)
          (CallFunction|FadeLockdownOut))))))

(fn UpdateScalabilitySequence
  (|CallParentFunction))

(fn ConstructionScript
  (|CallParentFunction))
