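; Distant-light actor for heist base lockdown, shown or hidden through the EnableVolume state.
; On BeginPlay it hides itself where cosmetics can run, and once playlist data is ready it listens for lockdown-initiated gameplay messages.
; Each message sets EnableVolume to whether its target tags exactly match this base's ID tag, and OnRep_EnableVolume updates visibility.
; LockdownOn and LockdownOff are server-called debug events that force the state.
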(event Custom|LockdownOn
  ;[1 1 1] Debug events (call from server)
  (Variables|Default|SetEnableVolume true))

(event Custom|LockdownOff
  ;[1 1 1] Debug events (call from server)
  (Variables|Default|SetEnableVolume false))

(event EventBeginPlay
  (|CallParentFunction)
  (Class|FortAsyncActionCurrentPlaylistReady|PlaylistDataReadyAsync
    (:then
      (Networking|CanExecuteCosmeticEvents
        (:True
          (Utilities|FlowControl|DoOnce false
            (:Completed
              (Rendering|SetActorHiddenInGame true))))))
    (:OnReady
      (bind (_asynctaskproxy _payload) (Messaging|RegisterGameplayMessageReceiver (Variables|Default|GetLockdownInitiatedChannelTag) "/Script/GameplayAbilities.GameplayEventData")
        (:OnMessageReceived
          (Variables|Default|SetEnableVolume (GameplayTags|HasTag (.TargetTags (Utilities|Struct|BreakGameplayEventData _payload)) (Variables|Default|GetHeistBaseIdTagConst) true)))))))

(fn OnRep_EnableVolume
  (Networking|CanExecuteCosmeticEvents
    (:True
      (Rendering|SetActorHiddenInGame (not (Variables|Default|GetEnableVolume))))))

(fn ConstructionScript
  (|CallParentFunction))
